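# Physics asset for the necklace on the F_MED_Winter_Hunter female player body, defining its simulated bodies and joints.
# Box bodies cover the dyn_ necklace and chain bones, a small base, and a larger spine_05 upper-torso volume.
# Both necklace pieces and the front chain constrain to the base with narrow twist and moderate swing limits.
# The back chain constrains more tightly to spine_05, and no constraint links the base to the spine.
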PhysicsAsset /BRCosmetics/Characters/Player/Female/Medium/Bodies/F_MED_Winter_Hunter/Meshes/F_MED_Winter_Hunter_Necklace_Physics
bodies {
 dyn_necklace_01 [{shapeType=Box, extentX=6.4111, extentY=8.8045, extentZ=3.1875, center=<2.316 0 0.6879>, rotation=<rot 11.2503 0 0>}]
 dyn_necklace_02 [{shapeType=Box, extentX=6.8486, extentY=16.367, extentZ=1.9375, center=<3.4131 -0.0001 -0.2708>, rotation=<rot -5.6251 0 0>}]
 dyn_necklace_chain_back [{shapeType=Box, extentX=2.5972, extentY=3.7252, extentZ=2.4913, center=<0.978 0 0>}]
 dyn_necklace_chain_front [{shapeType=Box, extentX=3.7861, extentY=8.8045, extentZ=3.375, center=<1.7184 -0.0001 0>, rotation=<rot -2.8126 0 0>}]
 dyn_necklaces_base [{shapeType=Box, extentX=0.8125, extentY=0.8125, extentZ=0.8125}]
 spine_05 [{shapeType=Box, extentX=23.125, extentY=9.3125, extentZ=25.6875, center=<-1.8123 4.4857 -0.0001>, rotation=<rot 0 -25.3122 0>}]
}
constraints [
 {bone1Name=dyn_necklace_chain_front, bone2Name=dyn_necklaces_base, swing1LimitDegrees=20, swing2LimitDegrees=30, twistLimitDegrees=2}
 {bone1Name=dyn_necklace_01, bone2Name=dyn_necklaces_base, swing1LimitDegrees=30, swing2LimitDegrees=20, twistLimitDegrees=2}
 {bone1Name=dyn_necklace_02, bone2Name=dyn_necklaces_base, swing1LimitDegrees=30, swing2LimitDegrees=20, twistLimitDegrees=2}
 {bone1Name=dyn_necklace_chain_back, bone2Name=spine_05, swing1LimitDegrees=10, swing2LimitDegrees=10, twistLimitDegrees=3}
]
