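# Physics asset for the hip/thigh pouches on the Winter Hunter female medium body cosmetic, used to give the pouches secondary motion.
# Capsule bodies on the front and back pouch bones are constrained to a small box body on their shared parent bone, with limited swing and very tight twist.
# A large capsule on the left thigh has no constraint and serves as a collision volume for the swinging pouches.
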PhysicsAsset /BRCosmetics/Characters/Player/Female/Medium/Bodies/F_MED_Winter_Hunter/Meshes/F_MED_Winter_Hunter_Pouches_Physics
bodies {
 dyn_pouch_back [{shapeType=Capsule, length=3.7101, radius=3.0575, center=<4.9959 0 0>, rotation=<rot 90 0 0>}]
 dyn_pouch_front [{shapeType=Capsule, length=3.5998, radius=4.0849, center=<5.2597 0 1.0071>, rotation=<rot 90 0 0>}]
 dyn_pouches_parent [{shapeType=Box, extentX=1.01, extentY=1.01, extentZ=1.01, rotation=<rot 0 -90 0>}]
 thigh_l [{shapeType=Capsule, length=21.875, radius=8.125, center=<-8.2302 0.0002 1.7191>, rotation=<rot 81.5619 179.9998 179.9999>}]
}
constraints [
 {bone1Name=dyn_pouch_back, bone2Name=dyn_pouches_parent, swing1LimitDegrees=30, swing2LimitDegrees=30, twistLimitDegrees=3}
 {bone1Name=dyn_pouch_front, bone2Name=dyn_pouches_parent, swing1LimitDegrees=30, swing2LimitDegrees=40, twistLimitDegrees=3}
]
