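# Post-process style cosmetic AnimBlueprint for the female Winter Hunter body, copying the attached parent mesh's pose and adding secondary motion.
# Thigh swing angles, remapped each update, drive additive scaling of both dyn_sub_thigh parent bones.
# A weighted spine and pelvis constraint positions the pouch base, then two chained RigidBody nodes simulate pouches and necklace with separate physics assets.
# Crouch events set a crouching alpha, which the pose graph never reads.
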AnimBlueprint /BRCosmetics/Characters/Player/Female/Medium/Bodies/F_MED_Winter_Hunter/Meshes/F_MED_Winter_Hunter_AnimBp
target_skeleton=&/F_MED_Winter_Hunter_Skeleton
graphs [
 AnimGraph {nodes [
   n1 = "Transform (Modify) Bone - Bone: dyn_sub_thigh_l_parent"(
    ComponentPose=Local To Component(LocalPose=Copy Pose From Mesh() {bUseAttachedParent=true})
    Alpha=Get legSwingDelta_l() -> (legSwingDelta_l:double)
    Translation="0,0,0"
    Rotation="0,0,0"
    Scale="4,2,4"
   ) {
    scaleMode=BMM_Additive
    scaleSpace=BCS_BoneSpace
    translationMode=BMM_Additive
    translationSpace=BCS_BoneSpace
    alphaBoolBlend {blendOption=Linear}
    boneToModify {boneName=dyn_sub_thigh_l_parent}
    scale=<4 2 4>
   }
   n2 = "Transform (Modify) Bone - Bone: dyn_sub_thigh_r_parent"(ComponentPose=@n1, Alpha=Get legSwingDelta_r() -> (legSwingDelta_r:double), Translation="0,0,0", Rotation="0,0,0", Scale="4,2,4") {
    scaleMode=BMM_Additive
    scaleSpace=BCS_BoneSpace
    translationMode=BMM_Additive
    translationSpace=BCS_BoneSpace
    alphaBoolBlend {blendOption=Linear}
    boneToModify {boneName=dyn_sub_thigh_r_parent}
    scale=<4 2 4>
   }
   n3 = "Constraint - dyn_pouches_base (PPPPP)"!(
    ComponentPose=@n2
    Alpha=1
    ConstraintWeights_0=0.2453
    ConstraintWeights_1:float=0.4023
    ConstraintWeights_2:float=0.0632
    ConstraintWeights_3:float=0.123
    ConstraintWeights_4:float=0.1661
   ) -> (Pose) {
    alphaBoolBlend {blendOption=Linear}
    alphaScaleBiasClamp {bInterpResult=true, interpSpeedDecreasing=3, interpSpeedIncreasing=3}
    boneToModify {boneName=dyn_pouches_base}
    constraintSetup [
     {targetBone {boneName=spine_01}}
     {targetBone {boneName=spine_02}}
     {targetBone {boneName=spine_03}}
     {targetBone {boneName=spine_04}}
     {targetBone {boneName=pelvis}}
    ]
    constraintWeights [1,1,1,1,1]
   }
   n4 = RigidBody(ComponentPose=@n3) {
    bClampLinearTranslationLimitToRefPose=true
    bOverrideWorldGravity=true
    overridePhysicsAsset=&/F_MED_Winter_Hunter_Pouches_Physics
    simulationSpace=BaseBoneSpace
    alphaBoolBlend {blendOption=Linear}
    alphaScaleBiasClamp {bInterpResult=true, interpSpeedDecreasing=5, interpSpeedIncreasing=3}
    baseBoneRef {boneName=root}
    componentAppliedLinearAccClamp=<4000 4000 4000>
    componentLinearAccScale=<0.25 0.25 0.25>
    componentLinearVelScale=<0.25 0.25 0.25>
    overrideWorldGravity {z=-980}
   }
   n5 = RigidBody(ComponentPose=@n4) {
    bClampLinearTranslationLimitToRefPose=true
    bOverrideWorldGravity=true
    overridePhysicsAsset=&/F_MED_Winter_Hunter_Necklace_Physics
    simulationSpace=BaseBoneSpace
    alphaBoolBlend {blendOption=Linear}
    alphaScaleBiasClamp {bInterpResult=true, interpSpeedDecreasing=5, interpSpeedIncreasing=5}
    baseBoneRef {boneName=root}
    componentAppliedLinearAccClamp=<6000 8000 8000>
    componentLinearAccScale=<0.5 0.5 0.5>
    componentLinearVelScale=<0.5 0.5 0.5>
    overrideWorldGravity {z=-980}
   }
   Output Pose(Result=Component To Local(ComponentPose=@n5)) {layerGroup=None}
  ]}
 EventGraph {nodes [
   Event BlueprintUpdateAnimation()
   n1 = GetOwnerSkeletalMeshComponent()
   > Set legSwingDelta_r(legSwingDelta_r:double=MapRangeClamped(Value=GetTwistAndSwingAngleOfDeltaRotationFromRefPose(self=@n1, BoneName=thigh_r).OutSwingAngle, InRangeA=150, InRangeB=75)) -> (Output_Get:double)
   > Set legSwingDelta_l(legSwingDelta_l:double=MapRangeClamped(Value=GetTwistAndSwingAngleOfDeltaRotationFromRefPose(self=@n1, BoneName=thigh_l).OutSwingAngle, InRangeA=43, InRangeB=110)) -> (Output_Get:double)
   Event OnBeginCrouching()
   > Set crouching alpha(crouching alpha:double=1) -> (Output_Get:double)
   Event OnEndCrouching()
   > Set crouching alpha(crouching alpha=0)
   TryGetPawnOwner()
  ]}
]
variables [
 crouching alpha {type=number}
 legSwingDelta_l {type=number}
 legSwingDelta_r {type=number}
]
