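# Physics asset for the face accessory part of the female medium Winter Hunter cosmetic body, providing secondary motion for its hat and two ponytails.
# Each ponytail is a six-capsule chain hanging from a box base, with limited swing and twist, tight at the root and looser mid-chain.
# The hat is a box body constrained to its own base.
# Unconstrained head, neck, clavicle, upper arm and upper spine bodies supply collision shapes around the chains.
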PhysicsAsset /BRCosmetics/Characters/Player/Female/Medium/Bodies/F_MED_Winter_Hunter/Meshes/Parts/F_MED_Winter_Hunter_ADS_FaceAcc_Physics
bodies {
 clavicle_l [{shapeType=Capsule, length=8.1562, radius=5.75, center=<5.6073 -0.4735 0.6504>, rotation=<rot 90 12.9074 10.0948>}]
 clavicle_r [{shapeType=Capsule, length=8.1562, radius=5.75, center=<-4.4924 0.5939 -0.1767>, rotation=<rot 90 0 0>}]
 dyn_hat_base [{shapeType=Box, extentX=0.5, extentY=0.5, extentZ=0.5}]
 dyn_hat_parent [{shapeType=Box, extentX=4.5728, extentY=7.4471, extentZ=1.6877, center=<2.3934 0 0>}]
 dyn_ponytail_l_01 [{shapeType=Capsule, length=6.4432, radius=2.6497, center=<4.708 -0.0002 0>, rotation=<rot 90 0 0>}]
 dyn_ponytail_l_02 [{shapeType=Capsule, length=6.321, radius=2.0963, center=<3.9216 0 -0.0005>, rotation=<rot 90 0 0>}]
 dyn_ponytail_l_03 [{shapeType=Capsule, length=3.0188, radius=2.1676, center=<2.7224 -0.0002 0>, rotation=<rot 90 0 0>}]
 dyn_ponytail_l_04 [{shapeType=Capsule, length=2.1164, radius=1.5304, center=<1.5074 0 0.0001>, rotation=<rot 90 0 0>}]
 dyn_ponytail_l_05 [{shapeType=Capsule, length=2.4303, radius=1.5444, center=<1.7735 0.053 -0.1011>, rotation=<rot 90 0 0>}]
 dyn_ponytail_l_06 [{shapeType=Capsule, length=3.3351, radius=1.0168, center=<2.6461 0 0>, rotation=<rot 90 0 0>}]
 dyn_ponytail_l_base [{shapeType=Box, extentX=1.0625, extentY=1.0625, extentZ=1.0625}]
 dyn_ponytail_r_01 [{shapeType=Capsule, length=5.2389, radius=2.4875, center=<4.708 -0.0002 0>, rotation=<rot 90 0 0>}]
 dyn_ponytail_r_02 [{shapeType=Capsule, length=6.4622, radius=2.7316, center=<3.802 0.9272 -0.2536>, rotation=<rot 90 0 -5.6251>}]
 dyn_ponytail_r_03 [{shapeType=Capsule, length=2.8081, radius=2.0989, center=<2.3518 0.25 0.1246>, rotation=<rot 90 0 0>}]
 dyn_ponytail_r_04 [{shapeType=Capsule, length=1.4469, radius=1.22, center=<1.5068 0.0074 -0.0421>, rotation=<rot 90 0 0>}]
 dyn_ponytail_r_05 [{shapeType=Capsule, length=3.2116, radius=1.3881, center=<1.7735 0.053 -0.1011>, rotation=<rot 90 0 0>}]
 dyn_ponytail_r_06 [{shapeType=Capsule, length=3.0538, radius=1.2043, center=<1.4731 0 0>, rotation=<rot 90 0 0>}]
 dyn_ponytail_r_base [{shapeType=Box, extentX=1.0625, extentY=1.0625, extentZ=1.0625}]
 head [{shapeType=Capsule, length=4.5, radius=5, center=<2.7362 1.3895 0>, rotation=<rot -90 0 0>}]
 neck_01 [{shapeType=Capsule, length=5.6562, radius=3.8438, center=<2.7362 1.3895 -0.2112>, rotation=<rot -90 0 0>}]
 spine_04 [
  {shapeType=Capsule, length=19.8125, radius=10.1875, center=<-11.4467 -0.3677 0.0875>, rotation=<rot 90 -0.0298 -0.03>}
  {shapeType=Box, extentX=20.875, extentY=7.5, extentZ=48.625, center=<14.3619 -7.8555 1.4241>}
 ]
 spine_05 [
  {shapeType=Box, extentX=20.5, extentY=9.8125, extentZ=23.3125, center=<-3.8444 3.8406 0>, rotation=<rot 0 -25.3112 0>}
  {shapeType=Box, extentX=10.1875, extentY=2.9375, extentZ=23.3125, center=<3.6203 5.5143 0>, rotation=<rot 0 -25.3105 0>}
 ]
 upperarm_l [{shapeType=Capsule, length=18.5938, radius=4.75, center=<8.8461 0 -0.0002>, rotation=<rot -90 -0.0099 0.0099>}]
 upperarm_r [{shapeType=Capsule, length=18.5938, radius=4.75, center=<-10.2461 0.0001 -0.0002>, rotation=<rot -90 -0.0297 0.0297>}]
}
constraints [
 {bone1Name=dyn_ponytail_l_01, bone2Name=dyn_ponytail_l_base, swing1LimitDegrees=5, swing2LimitDegrees=5, twistLimitDegrees=2}
 {bone1Name=dyn_ponytail_l_05, bone2Name=dyn_ponytail_l_04, swing1LimitDegrees=40, swing2LimitDegrees=60, twistLimitDegrees=6}
 {bone1Name=dyn_ponytail_l_02, bone2Name=dyn_ponytail_l_01, swing1LimitDegrees=30, swing2LimitDegrees=40, twistLimitDegrees=5}
 {bone1Name=dyn_ponytail_l_03, bone2Name=dyn_ponytail_l_02, swing1LimitDegrees=50, swing2LimitDegrees=40, twistLimitDegrees=2}
 {bone1Name=dyn_ponytail_l_04, bone2Name=dyn_ponytail_l_03, swing1LimitDegrees=50, swing2LimitDegrees=60, twistLimitDegrees=6}
 {bone1Name=dyn_ponytail_l_06, bone2Name=dyn_ponytail_l_05, swing1LimitDegrees=10, swing2LimitDegrees=40, twistLimitDegrees=6}
 {bone1Name=dyn_ponytail_r_01, bone2Name=dyn_ponytail_r_base, swing1LimitDegrees=5, swing2LimitDegrees=5, twistLimitDegrees=2}
 {bone1Name=dyn_ponytail_r_02, bone2Name=dyn_ponytail_r_01, swing1LimitDegrees=20, swing2LimitDegrees=20, twistLimitDegrees=2}
 {bone1Name=dyn_ponytail_r_03, bone2Name=dyn_ponytail_r_02, swing1LimitDegrees=50, swing2LimitDegrees=40, twistLimitDegrees=2}
 {bone1Name=dyn_ponytail_r_04, bone2Name=dyn_ponytail_r_03, swing1LimitDegrees=50, swing2LimitDegrees=50, twistLimitDegrees=2}
 {bone1Name=dyn_ponytail_r_05, bone2Name=dyn_ponytail_r_04, swing1LimitDegrees=40, swing2LimitDegrees=40, twistLimitDegrees=2}
 {bone1Name=dyn_ponytail_r_06, bone2Name=dyn_ponytail_r_05, swing1LimitDegrees=10, swing2LimitDegrees=10, twistLimitDegrees=2}
 {bone1Name=dyn_hat_parent, bone2Name=dyn_hat_base, swing1LimitDegrees=5, swing2LimitDegrees=20, twistLimitDegrees=2}
]
