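# Post-process-style face accessory AnimBP for the female Winter Hunter, copying the parent mesh pose, constraining the hat to the head and simulating ponytail physics.
# bIsTargeting blends between two RigidBody chains, ADS and default, each with bone offsets.
# The event graph resets dynamics when targeting toggles, lowers the ADS simulation alpha during skydiving or parachuting, and enables a default-chain frontend offset.
# PhysicsAsset_Swap is set but unread here.
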AnimBlueprint /BRCosmetics/Characters/Player/Female/Medium/Bodies/F_MED_Winter_Hunter/Meshes/Parts/F_MED_Winter_Hunter_FaceAcc_AnimBp
target_skeleton=&/F_MED_Winter_Hunter_FaceAcc_Skeleton
graphs [
 AnimGraph {nodes [
   n1 = RigidBody(ComponentPose=Local To Component(LocalPose=Use cached pose 'BraidCache'()), Alpha=Get Braid_Alpha() -> (Braid_Alpha:double)) {
    bClampLinearTranslationLimitToRefPose=true
    bOverrideWorldGravity=true
    overridePhysicsAsset=&/F_MED_Winter_Hunter_ADS_FaceAcc_Physics
    simulationSpace=BaseBoneSpace
    alphaBoolBlend {blendOption=Linear}
    alphaScaleBiasClamp {bInterpResult=true, interpSpeedDecreasing=5, interpSpeedIncreasing=5}
    baseBoneRef {boneName=spine_03}
    componentAppliedLinearAccClamp=<600 600 600>
    componentLinearAccScale=<0.6 0.6 0.6>
    componentLinearVelScale=<0.6 0.6 0.6>
    overrideWorldGravity {z=-980}
   }
   n2 = "Transform (Modify) Bone - Bone: dyn_ponytail_r_03"(ComponentPose=@n1, Alpha=1, Translation="0,0,-2", Rotation="0,0,0", Scale="1,1,1") {
    rotationMode=BMM_Additive
    rotationSpace=BCS_BoneSpace
    translationMode=BMM_Additive
    translationSpace=BCS_BoneSpace
    alphaBoolBlend {blendOption=Linear}
    alphaScaleBiasClamp {bInterpResult=true, interpSpeedDecreasing=3, interpSpeedIncreasing=3}
    boneToModify {boneName=dyn_ponytail_r_03}
    translation {z=-2}
   }
   n3 = "Transform (Modify) Bone - Bone: dyn_ponytail_l_base"(ComponentPose=@n2, Alpha=1, Translation="0,0,0", Rotation="0,5,0", Scale="1,1,1") {
    rotationMode=BMM_Additive
    rotationSpace=BCS_BoneSpace
    alphaBoolBlend {blendOption=Linear}
    alphaScaleBiasClamp {bInterpResult=true, interpSpeedDecreasing=3, interpSpeedIncreasing=3}
    boneToModify {boneName=dyn_ponytail_l_base}
    rotation {yaw=5}
   }
   n4 = "Transform (Modify) Bone - Bone: dyn_ponytail_r_base"(ComponentPose=@n3, Alpha=1, Translation="0,0,0", Rotation="8,0,0", Scale="1,1,1") {
    rotationMode=BMM_Additive
    rotationSpace=BCS_BoneSpace
    translationMode=BMM_Additive
    translationSpace=BCS_BoneSpace
    alphaBoolBlend {blendOption=Linear}
    alphaScaleBiasClamp {bInterpResult=true, interpSpeedDecreasing=3, interpSpeedIncreasing=3}
    boneToModify {boneName=dyn_ponytail_r_base}
    rotation {pitch=8}
   }
   n5 = RigidBody(ComponentPose=Local To Component(LocalPose=Use cached pose 'BraidCache'())) {
    bClampLinearTranslationLimitToRefPose=true
    bOverrideWorldGravity=true
    overridePhysicsAsset=&/F_MED_Winter_Hunter_FaceAcc_Physics
    simulationSpace=BaseBoneSpace
    alphaBoolBlend {blendOption=Linear}
    alphaScaleBiasClamp {bInterpResult=true, interpSpeedDecreasing=5, interpSpeedIncreasing=5}
    baseBoneRef {boneName=spine_05}
    componentAppliedLinearAccClamp=<600 600 600>
    componentLinearAccScale=<0.6 0.6 0.6>
    componentLinearVelScale=<0.6 0.6 0.6>
    overrideWorldGravity {z=-980}
   }
   n6 = "Transform (Modify) Bone - Bone: dyn_ponytail_r_03"(ComponentPose=@n5, Alpha=1, Translation="0,0,-2", Rotation="0,0,0", Scale="1,1,1") {
    rotationMode=BMM_Additive
    rotationSpace=BCS_BoneSpace
    translationMode=BMM_Additive
    translationSpace=BCS_BoneSpace
    alphaBoolBlend {blendOption=Linear}
    alphaScaleBiasClamp {bInterpResult=true, interpSpeedDecreasing=3, interpSpeedIncreasing=3}
    boneToModify {boneName=dyn_ponytail_r_03}
    translation {z=-2}
   }
   n7 = "Transform (Modify) Bone - Bone: dyn_ponytail_l_02"(ComponentPose=@n6, Alpha=Get IsInFrontend() -> (IsInFrontend:double), Translation="0,0,-1.75", Rotation="0,0,0", Scale="1,1,1") {
    rotationSpace=BCS_BoneSpace
    translationMode=BMM_Additive
    translationSpace=BCS_BoneSpace
    alphaBoolBlend {blendOption=Linear}
    boneToModify {boneName=dyn_ponytail_l_02}
    translation {z=-1.75}
   }
   Output Pose(Result=Blend Poses by bool(
     BlendPose_0=Component To Local(ComponentPose=@n4)
     BlendPose_1=Component To Local(ComponentPose=@n7)
     BlendTime_0=0.25
     BlendTime_1=0.25
     bActiveValue=Get bIsTargeting()
    ) {blendType=Linear}) {layerGroup=None}
   n8 = "Constraint - dyn_hat_base (P)"!(
    ComponentPose=Local To Component(LocalPose=Copy Pose From Mesh() {bCopyCurves=true, bUseAttachedParent=true})
    Alpha=1
    ConstraintWeights_0=1
    ConstraintWeights_1:float=0
   ) -> (Pose) {
    alphaBoolBlend {blendOption=Linear}
    boneToModify {boneName=dyn_hat_base}
    constraintSetup [{targetBone {boneName=head}}]
    constraintWeights [1,0]
   }
   Save cached pose 'BraidCache'(Pose=Component To Local(ComponentPose=@n8)) {cacheName=BraidCache}
  ]}
 EventGraph {nodes [
   Event BlueprintUpdateAnimation()
   > n1 = Sequence!(execute) -> (then_0, then_1, then_2:Exec)
   n2 = Get bIsSkydiveFloating()
   n3 = Get bIsSkyDiving()
   n4 = Get bIsParachuteOpened()
   n5 = Get bIsSkydiveDiveMode()
   > Set Braid_Alpha(Braid_Alpha:double=SelectFloat(A=0.25, B=1, bPickA=OR Boolean!(A=@n3, B=@n4, C:bool=@n2, D:bool=@n5) -> (ReturnValue))) -> (Output_Get:double)
   > n6 = Branch(Condition=Get bIsTargeting())
   > Do Once(Reset=@n8.then)
   > n7 = ResetDynamics(InTeleportType=ResetPhysics)
   Do Once(execute=@n6.else, Reset=@n7.then)
   > n8 = ResetDynamics(InTeleportType=ResetPhysics)
   Set PhysicsAsset_Swap(
    execute=@n1.then_1
    PhysicsAsset_Swap:bool=OR Boolean!(A=Get bIsRelaxedLevel1() -> (bIsRelaxedLevel1:bool), B=Get bIsRelaxedLevel2() -> (bIsRelaxedLevel2:bool), C=Get bIsTargeting()) -> (ReturnValue)
   ) -> (Output_Get:bool)
   Set IsInFrontend(execute=@n1.then_2, IsInFrontend:double=SelectFloat(A=1, B=0, bPickA=Get bIsInFrontEnd())) -> (Output_Get:double)
   TryGetPawnOwner()
  ]}
]
variables [
 Braid_Alpha {type=number}
 IsInFrontend {type=number}
 PhysicsAsset_Swap {type=boolean}
]
